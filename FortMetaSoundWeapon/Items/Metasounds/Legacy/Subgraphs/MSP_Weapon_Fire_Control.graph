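# Legacy MetaSound patch that controls weapon fire timing, outputting gunshot and stop triggers, firing state and ammo count.
# In auto-fire, a Bool trigger route starts or stops a FireRate repeat, with delayed stop and accumulated retrigger-plus-stop restarting it.
# A counter derives remaining ammo from StartingAmmo to gate repeated shots.
# Without auto-fire, trigger and retrigger emit single shots directly; OnOutOfAmmo is unconnected.
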MetaSoundPatch /FortMetaSoundWeapon/Items/Metasounds/Legacy/Subgraphs/MSP_Weapon_Fire_Control
inputs [
 bAutoFire {type=Bool}
 FireRate {type=Time}
 OnRetrigger {type=Trigger}
 OnStop {type=Trigger}
 OnTrigger {type=Trigger}
 StartingAmmo {type=Int32}
 StopTriggerDelay {type=Time}
]
interfaces [MetaSound:1.0]
outputs [
 bIsFiring {type=Bool}
 CurrentAmmoOut {type=Int32}
 OnGunshotOut {type=Trigger}
 OnOutOfAmmo {type=Trigger}
 TriggerOutStop {type=Trigger}
]
pages [Default {
  input_defaults {
   FireRate=0.1
   OnRetrigger=null
   OnStop=null
   OnTrigger=null
   StartingAmmo=30
   StopTriggerDelay=0.03
   bAutoFire="<MetaSoundBuilderResult.SUCCEEDED: 0>"
  }
  nodes [
   AbsoluteValue.Abs:Int32()
   InitVariable.Trigger()
   PrintLogPrintLog_0 = Print Log.Print Log:Int32() {also {defaults {Label="DebugMSSWeapon: "}}}
   PrintLogPrintLog_1 = Print Log.Print Log:String() {also {defaults {Label="DebugMSSWeapon: ", Value To Log=Play}}}
   PrintLogPrintLog_2 = Print Log.Print Log:String() {also {defaults {Label="DebugMSSWeapon: ", Value To Log=Stop}}}
   PrintLogPrintLog_3 = Print Log.Print Log:String() {also {defaults {Label="DebugMSSWeapon: ", Value To Log=Retrigger}}}
   TriggerDelay_1 = UE.Trigger Delay(In=^OnStop, Delay Time=^StopTriggerDelay)
   n1 = TriggerCompare.Clamp:Bool(Compare=^OnRetrigger, A=^bAutoFire) {also {defaults {B="<MetaSoundBuilderResult.SUCCEEDED: 0>"}}}
   TriggerAccumulatorTriggerAccumulate2_0 = "TriggerAccumulator.Trigger Accumulate (2)"(In 0=@n1.True, In 1=@TriggerDelay_1.Out) {also {defaults {Auto Reset="<MetaSoundBuilderResult.SUCCEEDED: 0>"}}}
   TriggerDelay_0 = UE.Trigger Delay(In=^OnTrigger) {also {defaults {Delay Time=0.01}}}
   n2 = TriggerCompare.Clamp:Bool(A=^bAutoFire, Compare=@TriggerDelay_0.Out) {also {defaults {B="<MetaSoundBuilderResult.SUCCEEDED: 0>"}}}
   TriggerRouteTriggerRouteBool3_0 = "TriggerRoute.Trigger Route (Bool, 3):Bool"(Set 1=UE.Reroute(Value=@TriggerDelay_1.Out).Value, Set 2=@TriggerAccumulatorTriggerAccumulate2_0.Out, Set 0=@n2.True) {also {defaults {Value 0="<MetaSoundBuilderResult.SUCCEEDED: 0>", Value 2="<MetaSoundBuilderResult.SUCCEEDED: 0>"}}}
   n3 = InitVariable.Bool()
   n4 = TriggerCompare.Clamp:Bool(
    Compare=@TriggerRouteTriggerRouteBool3_0.On Set
    A=VariableAccessor.Bool(Variable=VariableMutator.Bool(Variable=@n3.Variable, Value=@TriggerRouteTriggerRouteBool3_0.Value).Variable).Value
   ) {also {defaults {B="<MetaSoundBuilderResult.SUCCEEDED: 0>"}}}
   n5 = UE.Reroute(Value=UE.Reroute(Value=@n4.False).Value)
   Reroute_2 = UE.Reroute(Value=@n5.Value)
   TriggerOnValueChange_0 = UE.Trigger On Value Change:Int32(Value=^StartingAmmo)
   TriggerAnyTriggerAny2_2 = "TriggerAny.Trigger Any (2)"(In 0=@n5.Value, In 1=@TriggerOnValueChange_0.Trigger)
   TriggerCounter_0 = UE.Trigger Counter(Reset=@TriggerAnyTriggerAny2_2.Out, In=UE.TriggerRepeat(Period=^FireRate, Stop=@n4.False, Start=@n4.True).RepeatOut)
   TriggerAnyTriggerAny2_1 = "TriggerAny.Trigger Any (2)"(In 0=@n1.False, In 1=@n2.False)
   n6 = VariableAccessor.Trigger(Variable=VariableMutator.Trigger(Variable=InitVariable.Trigger().Variable, Value=@TriggerAnyTriggerAny2_1.Out).Variable)
   TriggerRouteTriggerRouteInt322_0 = "TriggerRoute.Trigger Route (Int32, 2):Int32"(
    Value 1=^StartingAmmo
    Set 0=UE.Reroute(Value=@TriggerCounter_0.On Trigger).Value
    Value 0=UE.Subtract:Int32(
     PrimaryOperand=UE.Add:Int32(PrimaryOperand=^StartingAmmo) {also {defaults {AdditionalOperands=1}}}.Out
     AdditionalOperands=@TriggerCounter_0.Count
    ).Out
    Set 1=VariableAccessor.Trigger(Variable=@n6.Variable).Value
   )
   n7 = VariableMutator.Int32(Variable=InitVariable.Int32().Variable, Value=@TriggerRouteTriggerRouteInt322_0.Value)
   n8 = VariableMutator.Trigger(Variable=InitVariable.Trigger().Variable, Value=@TriggerCounter_0.On Trigger)
   TriggerAnyTriggerAny2_0 = "TriggerAny.Trigger Any (2)"(
    In 0=TriggerCompare.Clamp:Int32(A=VariableAccessor.Int32(Variable=@n7.Variable).Value, Compare=VariableAccessor.Trigger(Variable=@n8.Variable).Value) {also {defaults {B=0, Type=5}}}.True
    In 1=@n6.Value
   )
   VariableMutator.Trigger(Variable=InitVariable.Trigger().Variable)
  ]
  links [
   {dst="@TriggerOutStop", src=Reroute_2.Value}
   {dst="@OnGunshotOut", src=TriggerAnyTriggerAny2_0.Out}
   {dst="@bIsFiring", src=TriggerRouteTriggerRouteBool3_0.Value}
   {dst="@CurrentAmmoOut", src=TriggerRouteTriggerRouteInt322_0.Value}
  ]
 }]
