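# MetaSound patch for the mechanical layer of weapon fire sounds, outputting mono audio from an array of mech wave assets.
# OnPlay passes a bEnable check, then waits a random delay between RandomTimeMin and RandomTimeMax.
# A trigger sequence alternates shots between two identical random-sound players with pitch and volume ranges, which are mixed together so consecutive sounds can overlap.
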MetaSoundPatch /FortMetaSoundWeapon/Items/Metasounds/Legacy/Subgraphs/MSP_Weapon_Fire_Mech
inputs [
 bEnable {type=Bool}
 MaxPitch {type=Float}
 MaxVolume {type=Float}
 MechLayerSounds {type=WaveAsset:Array}
 MinPitch {type=Float}
 MinVolume {type=Float}
 OnPlay {type=Trigger}
 RandomTimeMax {type=Time}
 RandomTimeMin {type=Time}
]
interfaces [MetaSound:1.0]
outputs [Out Mono {type=Audio}]
pages [Default {
  input_defaults {
   MaxPitch=1.05
   MaxVolume=1.05
   MinPitch=0.95
   MinVolume=0.95
   OnPlay=null
   RandomTimeMax=0
   RandomTimeMin=0
   bEnable="<MetaSoundBuilderResult.SUCCEEDED: 0>"
   MechLayerSounds []
  }
  nodes [
   n1 = UE.RandomTime(
    Max=^RandomTimeMax
    Min=^RandomTimeMin
    Next=TriggerCompare.Clamp:Bool(Compare=^OnPlay, A=^bEnable) {also {defaults {B="<MetaSoundBuilderResult.SUCCEEDED: 0>"}}}.True
   )
   TriggerDelay_0 = UE.Trigger Delay(In=@n1.On Next, Delay Time=@n1.Value)
   TriggerSequence2_0 = "UE.Trigger Sequence (2)"(In=@TriggerDelay_0.Out)
   AudioMixerAudioMixerMono2_0 = "AudioMixer.Audio Mixer (Mono, 2)"(
    In 0=&MSS_PlayRandomSound_Mono(
     MaxPitch=^MaxPitch
     MaxVolume=^MaxVolume
     Sounds=^MechLayerSounds
     MinPitch=^MinPitch
     MinVolume=^MinVolume
     UE.Source.OnPlay=@TriggerSequence2_0.Out 0
    )."UE.OutputFormat.Mono.Audio:0"
    In 1=&MSS_PlayRandomSound_Mono(
     MaxPitch=^MaxPitch
     MaxVolume=^MaxVolume
     Sounds=^MechLayerSounds
     MinPitch=^MinPitch
     MinVolume=^MinVolume
     UE.Source.OnPlay=@TriggerSequence2_0.Out 1
    )."UE.OutputFormat.Mono.Audio:0"
   )
  ]
  links [{dst="@Out Mono", src=AudioMixerAudioMixerMono2_0.Out}]
 }]
