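# A MetaSound patch that plays a randomized sweetener layer for weapon fire sounds and outputs it in stereo.
# When enabled, OnPlay picks a random delay within the time range, then a trigger sequence alternates between two random-sound players, presumably so consecutive shots overlap.
# Both players share the wave array and pitch and volume ranges, and only the first two channels of their 5.1 outputs feed the stereo mixer.
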MetaSoundPatch /FortMetaSoundWeapon/Items/Metasounds/Legacy/Subgraphs/MSP_Weapon_Fire_Sweetener
inputs [
 bEnable {type=Bool}
 MaxPitch {type=Float}
 MaxVolume {type=Float}
 MinPitch {type=Float}
 MinVolume {type=Float}
 OnPlay {type=Trigger}
 RandomTimeMax {type=Time}
 RandomTimeMin {type=Time}
 SweetenerLayerSounds {type=WaveAsset:Array}
]
interfaces [MetaSound:1.0]
outputs [
 Out Left {type=Audio}
 Out Right {type=Audio}
]
pages [Default {
  input_defaults {
   MaxPitch=1.05
   MaxVolume=1.05
   MinPitch=0.95
   MinVolume=0.95
   OnPlay=null
   RandomTimeMax=0
   RandomTimeMin=0
   bEnable="<MetaSoundBuilderResult.SUCCEEDED: 0>"
   SweetenerLayerSounds []
  }
  nodes [
   n1 = UE.RandomTime(
    Max=^RandomTimeMax
    Min=^RandomTimeMin
    Next=TriggerCompare.Clamp:Bool(Compare=^OnPlay, A=^bEnable) {also {defaults {B="<MetaSoundBuilderResult.SUCCEEDED: 0>"}}}.True
   )
   TriggerDelay_0 = UE.Trigger Delay(In=@n1.On Next, Delay Time=@n1.Value)
   TriggerSequence2_0 = "UE.Trigger Sequence (2)"(In=@TriggerDelay_0.Out)
   n2 = /FortMetasound/Sources/MSS_PlayRandomSound_5_1(
    MaxPitch=^MaxPitch
    MaxVolume=^MaxVolume
    MinPitch=^MinPitch
    MinVolume=^MinVolume
    Sounds=^SweetenerLayerSounds
    UE.Source.OnPlay=@TriggerSequence2_0.Out 0
   )
   n3 = &MSS_PlayRandomSound_5_1(
    MaxPitch=^MaxPitch
    MaxVolume=^MaxVolume
    MinPitch=^MinPitch
    MinVolume=^MinVolume
    Sounds=^SweetenerLayerSounds
    UE.Source.OnPlay=@TriggerSequence2_0.Out 1
   )
   AudioMixerAudioMixerStereo2_0 = "AudioMixer.Audio Mixer (Stereo, 2)"(
    In 0 L=@n2."UE.OutputFormat.5dot1.Audio:0"
    In 0 R=@n2."UE.OutputFormat.5dot1.Audio:1"
    In 1 L=@n3."UE.OutputFormat.5dot1.Audio:0"
    In 1 R=@n3."UE.OutputFormat.5dot1.Audio:1"
   )
  ]
  links [{dst="@Out Left", src=AudioMixerAudioMixerStereo2_0.Out L}, {dst="@Out Right", src=AudioMixerAudioMixerStereo2_0.Out R}]
 }]
