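# MetaSound source that plays a stereo low-ammo weapon fire layer, growing louder as the magazine runs out.
# On play, it compares current ammo to the low-ammo threshold and, on a pass, triggers a random sound with pitch and volume variation through a nested 5.1 player.
# The player's first two channels are scaled by a gain mapped from remaining ammo.
# It finishes when the sound ends, or immediately if the check fails.
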MetaSoundSource /FortMetaSoundWeapon/Items/Metasounds/Legacy/Subgraphs/MSS_Weapon_Fire_Low
inputs [
 CurrentAmmo {type=Int32}
 FireLowAmmoSounds {type=WaveAsset:Array}
 LowAmmoThreshold {type=Int32}
 MaxPitch {type=Float}
 MaxVolume {type=Float}
 MinPitch {type=Float}
 MinVolume {type=Float}
 UE.Source.OnPlay {type=Trigger}
]
interfaces [UE.OutputFormat.Stereo:1.0, UE.Source.OneShot:1.0, UE.Source:1.1]
outputs [
 "UE.OutputFormat.Stereo.Audio:0" {type=Audio}
 "UE.OutputFormat.Stereo.Audio:1" {type=Audio}
 UE.Source.OneShot.OnFinished {type=Trigger}
]
pages [Default {
  input_defaults {CurrentAmmo=1, LowAmmoThreshold=1, MaxPitch=1.05, MaxVolume=1.05, MinPitch=0.95, MinVolume=0.95, UE.Source.OnPlay=null, FireLowAmmoSounds []}
  nodes [
   n1 = TriggerCompare.Clamp:Int32(A=^CurrentAmmo, B=^LowAmmoThreshold, Compare=^UE.Source.OnPlay) {also {defaults {Type=4}}}
   n2 = /FortMetasound/Sources/MSS_PlayRandomSound_5_1(
    MaxPitch=^MaxPitch
    MaxVolume=^MaxVolume
    MinPitch=^MinPitch
    MinVolume=^MinVolume
    Sounds=UE.Reroute(Value=^FireLowAmmoSounds).Value
    UE.Source.OnPlay=@n1.True
   )
   n3 = MapRange.MapRange:Float(In=Convert.Float:Int32(Int32=^CurrentAmmo).Float, In Range A=Convert.Float:Int32(Int32=^LowAmmoThreshold).Float) {also {defaults {In Range B=0, Out Range A=0.3}}}
   Multiply_0 = UE.Multiply:Audio by Float(PrimaryOperand=@n2."UE.OutputFormat.5dot1.Audio:1", AdditionalOperands=@n3.Out Value)
   Multiply_1 = UE.Multiply:Audio by Float(PrimaryOperand=@n2."UE.OutputFormat.5dot1.Audio:0", AdditionalOperands=@n3.Out Value)
   TriggerAnyTriggerAny2_0 = "TriggerAny.Trigger Any (2)"(In 1=@n2.UE.Source.OneShot.OnFinished, In 0=UE.Reroute(Value=UE.Reroute(Value=@n1.False).Value).Value)
  ]
  links [
   {dst="@UE.OutputFormat.Stereo.Audio:1", src=Multiply_0.Out}
   {dst="@UE.OutputFormat.Stereo.Audio:0", src=Multiply_1.Out}
   {dst="@UE.Source.OneShot.OnFinished", src=TriggerAnyTriggerAny2_0.Out}
  ]
 }]
properties {
 assetImportData=&self:AssetImportData
 numChannels=2
 outputFormat=Stereo
 sampleRate=48000
}
